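; Animation notify for the Umbramolt Kipper's final-hit knockback, which tags the animated Fortnite player pawn.
; On notify, it casts the mesh component's owner to a Fort player pawn and adds its configurable Gameplay Tags as loose tags.
; It only applies the tags and prints a debug confirmation; no removal or knockback force happens in this graph.
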(fn Received_Notify
  (params (MeshComp :type Skeletal%20Mesh%20Component%20Object%20Reference) (Animation :type Anim%20Sequence%20Base%20Object%20Reference) (EventReference :type Anim%20Notify%20Event%20Reference%20Structure%20%28by%20ref%29))
  (bind _asfort_player_pawn (Utilities|Casting|CastToFortPlayerPawn (Components|GetOwner (Variables|GetMeshComp))))
  (Ability|GameplayEffect|AddLooseGameplayTags _asfort_player_pawn (Variables|Default|GetGameplay%20Tags) true)
  (Development|PrintString "Applied knockback tag" true true "(R=0.013441,G=1.000000,B=0.182678,A=1.000000)" 1e+01)
  (return false))
